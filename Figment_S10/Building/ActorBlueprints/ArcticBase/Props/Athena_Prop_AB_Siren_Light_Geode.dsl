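; Arctic base siren light prop that drives a red alert effect around a calendar-scheduled rocket launch countdown.
; When match-ready, the Pre-Launch Klaxons event sets a replicated server-time launch timestamp, and its end sets a launched flag after a 30-second timer.
; On clients, RepNotifies activate the alert VFX if at most ten minutes remain, with the spotlight hidden.
; After launch, they deactivate the VFX and unhide the spotlight.
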(event EventBeginPlay
  (Class|FortAsyncActionAuthorityMatchReady|MatchReadyAsync
    (:OnReady
      (bind (_eventname _timeuntilend) (MCPContext|WatchCalendarEvent (Variables|Default|GetPre-Launch%20Klaxons))
        (:OnEventActive
          (Variables|Default|SetTime%20Until%20Countdown%20End _timeuntilend)
          (CallFunction|ContinueCountdownEvent))
        (:OnEventEnded
          (Utilities|Time|SetTimerbyEvent (delegate Custom|CustomEvent_0) 3e+01))))))

(event Custom|CustomEvent_0
  (Variables|Default|SetIsRKTLaunchd true))

(event Custom|Continue%20CountdownEvent
  (Variables|Default|SetServer%20World%20Seconds%20at%20Launch (+ (GameState|GetServerWorldTimeSeconds (Game|GetGameState)) (Math|Timespan|GetTotalSeconds (Variables|Default|GetTime%20Until%20Countdown%20End)))))

(fn OnRep_IsRKTLaunchd
  (Networking|SwitchHasAuthority
    (:Remote
      (if (Variables|Default|GetIsRKTLaunchd)
        (Rendering|SetVisibility (Variables|Debug|GetStaticMeshComponent) true false)
        (Components|Activation|Deactivate (Variables|Default|GetRed%20Alert%20VFX))
        (Development|SetHiddeninGame (Variables|Default|GetSpotLight) false false)))))

(fn OnRep_ServerWorldSecondsatLaunch
  (Networking|SwitchHasAuthority
    (:Remote
      (bind _output_get (Variables|Default|SetSeconds%20Remaining (Math|Float|Truncate (- (Variables|Default|GetServer%20World%20Seconds%20at%20Launch) (GameState|GetServerWorldTimeSeconds (Game|GetGameState))))))
      (if (<= _output_get 600)
        (Rendering|SetVisibility (Variables|Debug|GetStaticMeshComponent) true false)
        (Components|Activation|Activate (Variables|Default|GetRed%20Alert%20VFX) false)
        (Development|SetHiddeninGame (Variables|Default|GetSpotLight) true false)
        (else
          (Components|Activation|Deactivate (Variables|Default|GetRed%20Alert%20VFX))
          (Development|SetHiddeninGame (Variables|Default|GetSpotLight) true false)))
      (return))))

(fn ConstructionScript
  (|CallParentFunction))
